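# First-person music cue for the Let's Play emote, using 1P emote music class and concurrency settings and a send to a music audio-analysis submix.
# A licensed-content switcher selects between looping silence and the full mix.
# The mixer layers immediate foley with a delayed fireworks launch, an enveloped fireworks loop that swells then tapers, and the music loop, which fades in quickly just after the launch.
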SoundCue /BRCosmetics/Sounds/Emotes/Lets_Play/Emote_Lets_Play_Music_Cue
root=FortSoundNodeLicensedContentSwitcher_0
nodes [
 {id=Delay_1, children [Enveloper_3], properties {delayMax=1.8, delayMin=1.8}}
 {id=Delay_2, children [WavePlayer_4], properties {delayMax=1.8, delayMin=1.8}}
 {id=Delay_3, children [Enveloper_2], properties {delayMax=1.86, delayMin=1.86}}
 {
  id=Enveloper_2
  children [WavePlayer_1]
  properties {pitchCurve {editorCurveData {keys [{time=0}]}}, volumeCurve {editorCurveData {keys [<at 0 0.01>,<at 0.25 1>]}}}
 }
 {
  id=Enveloper_3
  children [WavePlayer_3]
  properties {pitchCurve {editorCurveData {keys [{time=0}]}}, volumeCurve {editorCurveData {keys [<at 0 0>,<at 3 1>,<at 10 0.3>,<at 25 0.15>]}}}
 }
 {id=FortSoundNodeLicensedContentSwitcher_0, class=FortSoundNodeLicensedContentSwitcher, children [WavePlayer_0,Mixer_1,Mixer_1]}
 {id=Mixer_1, children [Delay_3,Delay_1,Delay_2,WavePlayer_5], properties {inputVolume [1,1,1,0.75]}}
 {id=WavePlayer_0, properties {bLooping=true, soundWaveAssetPtr=/Game/Sounds/Fort_Test_Sounds/5_seconds_silence}}
 {id=WavePlayer_1, properties {bLooping=true, soundWaveAssetPtr=&/Emote_Lets_Play_Loop}}
 {id=WavePlayer_3, properties {bLooping=true, soundWaveAssetPtr=/Game/Athena/Sounds/Events/Papaya/SFX/Papaya_Fireworks_Loop_02}}
 {id=WavePlayer_4, properties {soundWaveAssetPtr=/Game/Athena/Sounds/Events/Papaya/SFX/Papaya_Fireworks_Launch_01}}
 {id=WavePlayer_5, properties {soundWaveAssetPtr=&/Emote_Lets_Play_Foley}}
]
properties {
 audioPropertiesSheet=/BRCosmetics/Sounds/Settings/PropertySheets/Emotes/APS_Cosmetic_Emote_Music_1P
 maxDistance=2097152
 soundClassObject=/Game/Sounds/FortSoundClassesAndModes/Fort_Music_Emote_1P_PSM
 volumeMultiplier=1
 concurrencyOverrides {maxCount=3, resolutionRule=StopOldest}
 concurrencySet [/Game/Sounds/Concurrency_Rules/Athena_EmoteMusic_1P_Conc]
 soundSubmixSends [{minSendLevel=0, soundSubmix=&Fortnite_AudioAnalysis_Music}]
}
