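# Third-person sound cue for the Lets Play emote, attenuated and played under emote music concurrency rules.
# A licensed content switcher selects either looping silence or a four-layer mix.
# The mix starts foley immediately, then after a short delay adds a fireworks launch, the looping emote track and a fireworks loop.
# Envelopes quickly fade in the track, while the fireworks loop swells, then tapers to a low level.
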SoundCue /BRCosmetics/Sounds/Emotes/Lets_Play/SC_Emote_Lets_Play_3P_Cue
root=FortSoundNodeLicensedContentSwitcher_0
nodes [
 {id=Delay_0, children [Enveloper_1], properties {delayMax=1.8, delayMin=1.8}}
 {id=Delay_1, children [WavePlayer_2], properties {delayMax=1.8, delayMin=1.8}}
 {id=Delay_2, children [Enveloper_0], properties {delayMax=1.86, delayMin=1.86}}
 {
  id=Enveloper_0
  children [WavePlayer_0]
  properties {pitchCurve {editorCurveData {keys [{time=0}]}}, volumeCurve {editorCurveData {keys [<at 0 0.01>,<at 0.25 1>]}}}
 }
 {
  id=Enveloper_1
  children [WavePlayer_1]
  properties {pitchCurve {editorCurveData {keys [{time=0}]}}, volumeCurve {editorCurveData {keys [<at 0 0>,<at 3 1>,<at 10 0.3>,<at 25 0.15>]}}}
 }
 {id=FortSoundNodeLicensedContentSwitcher_0, class=FortSoundNodeLicensedContentSwitcher, children [WavePlayer_4,WavePlayer_4,Mixer_0]}
 {id=Mixer_0, children [Delay_2,Delay_0,Delay_1,WavePlayer_3], properties {inputVolume [1,1,1,0.75]}}
 {id=WavePlayer_0, properties {bLooping=true, soundWaveAssetPtr=&/Emote_Lets_Play_Loop}}
 {id=WavePlayer_1, properties {bLooping=true, soundWaveAssetPtr=/Game/Athena/Sounds/Events/Papaya/SFX/Papaya_Fireworks_Loop_02}}
 {id=WavePlayer_2, properties {soundWaveAssetPtr=/Game/Athena/Sounds/Events/Papaya/SFX/Papaya_Fireworks_Launch_01}}
 {id=WavePlayer_3, properties {soundWaveAssetPtr=&/Emote_Lets_Play_Foley}}
 {id=WavePlayer_4, properties {bLooping=true, soundWaveAssetPtr=/Game/Sounds/Fort_Test_Sounds/5_seconds_silence}}
]
properties {
 attenuationSettings=&Fort_Emotes_Athena_Att
 maxDistance=1600
 soundClassObject=/Game/Sounds/FortSoundClassesAndModes/Fort_Music_Emote_3P
 virtualizationMode=PlayWhenSilent
 volumeMultiplier=1
 concurrencyOverrides {resolutionRule=PreventNew}
 concurrencySet [/Game/Sounds/Concurrency_Rules/Athena_EmoteMusic_Conc]
 soundSubmixSends [{minSendLevel=0, soundSubmix=&Fortnite_AudioAnalysis_Music}]
}
